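# Translucent, unlit, two-sided material for Niagara ribbon swing trails.
# A U-offset mask on UV1 remaps and overlays (multiply or add) a gradient texture, tinted by particle color for emissive.
# Opacity combines particle and gradient alpha with power-curve fades along flipped U, with clamped exponents and optional mask erosion.
# Three dynamic parameter sets drive mask blending, contrast, offset and fade falloff per particle, while channel parameters pick separate color and alpha mask channels.
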Material /Game/Effects/Materials/SwingTrails/M_SwingTrails_Standard
graph {
 properties {
  bAutomaticallySetUsageInEditor=false
  bCastRayTracedShadows=false
  bUsedWithNiagaraRibbons=true
  blendMode=BLEND_Translucent
  shadingModel=MSM_Unlit
  twoSided=true
  usedShadingModels=Unlit
 }
 nodes [
  n1 = ParticleColor()
  n2 = DynamicParameter() {parameterIndex=1, defaultValue=<rgba 1 1 0 1>, paramNames [Mask Contrast, Mask Alpha Contrast, Mask U Offset, Mask Overlay Strength]}
  n3 = Multiply(A=@n2.Mask Overlay Strength, B=NamedRerouteUsage())
  n4 = OneMinus(None=TextureCoordinate())
  n5 = ComponentMask(None=@n4) {g=true}
  n6 = Constant()
  n7 = DynamicParameter() {
   desc="Mask Alpha Lerp: This Lerps the channel chosen into the alpha \r\n0: gradient \r\n1: multiplies mask into gradient\r\n\r\nMask Pre Color: Lerps between a gradient mapping V of trail to the mask. the color gradient will be applied to the 0-1 values of this\r\n\r\nMask Post color Mode: This dictates how the mask will be treated if Mask Post Color is used. \r\n0: Multiply mask \r\n1: Add Mask\r\n\r\nMask Post Color Lerp: This controls if the mask should also be applied after the coloring"
   defaultValue=<rgba 1 1 1 0>
   paramNames [Mask Alpha Lerp, Mask Lerp, "Mask Overlay Mode | Mult <> Add", Mask Overlay Lerp]
  }
  n8 = TextureSample(
   UVs=LinearInterpolate(A=AppendVector(A=@n5, B=@n6), B=NamedRerouteUsage(), Alpha=@n7.Mask Lerp) {constAlpha=1}
   Tex=TextureObjectParameter() {group=Base, parameterName=Gradient Texture, texture=/Game/Effects/Materials/SwingTrails/Textures/T_FX_ColorRamp_Linear}
  )
  n9 = DynamicParameter() {defaultValue=<rgba 1 1 1 0>, paramNames [Mask Alpha Lerp, Mask Lerp, "Mask Overlay Mode | Mult <> Add", Mask Overlay Lerp]}
  n10 = LinearInterpolate(A=Multiply(A=@n8.RGB, B=@n3), B=Add(A=@n3, B=@n8.RGB), Alpha=@n9."Mask Overlay Mode | Mult <> Add")
  MaterialFunctionCall(Emissive=Multiply(A=@n1.RGB, B=LinearInterpolate(A=@n8.RGB, B=@n10, Alpha=@n9.Mask Overlay Lerp) {constAlpha=0})) {materialFunction=&MF_FX_EmissiveAdjustments}
  n11 = DynamicParameter() {parameterIndex=1, defaultValue=<rgba 1 1 0 1>, paramNames [Mask Contrast, Mask Alpha Contrast, Mask U Offset, Mask Overlay Strength]}
  n12 = Constant()
  n13 = TextureSampleParameter2D(UVs=Add(A=TextureCoordinate() {coordinateIndex=1}, B=AppendVector(A=@n11.Mask U Offset, B=@n12))) {group=Mask, parameterName=Mask Texture, texture=/Game/Effects/Fort_Effects/Textures/T_InkySmokeTrail}
  n14 = DynamicParameter() {parameterIndex=1, defaultValue=<rgba 1 1 0 1>, paramNames [Mask Contrast, Mask Alpha Contrast, Mask U Offset, Mask Overlay Strength]}
  NamedRerouteDeclaration(None=MaterialFunctionCall(In=ChannelMaskParameter(None=@n13.RGBA) {group=Mask, parameterName=Mask Channel, sortPriority=33}.R, Contrast=@n14.Mask Contrast) {materialFunction=&CheapContrast}.Result) {name=Overlay Mask, nodeColor=<rgba 0 0.9297 1 1>}
  n15 = DynamicParameter() {parameterIndex=2, defaultValue=<rgba 0 1 0 0>, paramNames [Leading Edge Pow, Alpha Pow, Param3, Param4]}
  n16 = ScalarParameter() {
   defaultValue=1
   desc=This locks the power from going below a specific threshold.
   group=Base
   parameterName=Power Min Threshold
   sortPriority=33
  }
  n17 = ComponentMask(None=@n4) {r=true}
  n18 = DynamicParameter() {parameterIndex=2, defaultValue=<rgba 0 1 0 0>, paramNames [Leading Edge Pow, Alpha Pow, Param3, Param4]}
  n19 = ScalarParameter() {
   defaultValue=1
   desc=This locks the power from going below a specific threshold.
   group=Base
   parameterName=Power Min Threshold
   sortPriority=33
  }
  n20 = If(A=@n18.Leading Edge Pow, B=@n19, "A > B"=@n18.Leading Edge Pow, "A < B"=@n19) {constB=0.1}
  n21 = Multiply(
   A=Power(Base=@n17, Exp=If(A=@n15.Alpha Pow, B=@n16, "A > B"=@n15.Alpha Pow, "A < B"=@n16) {constB=0.1})
   B=OneMinus(None=Power(Base=@n17, Exp=@n20) {constExponent=20})
  )
  n22 = Add(A=@n21, B=@n21)
  n23 = DynamicParameter() {parameterIndex=1, defaultValue=<rgba 1 1 0 1>, paramNames [Mask Contrast, Mask Alpha Contrast, Mask U Offset, Mask Overlay Strength]}
  n24 = MaterialFunctionCall(In=ChannelMaskParameter(None=@n13.RGBA) {group=Mask, parameterName=Mask Alpha Channel, sortPriority=34}.R, Contrast=@n23.Mask Alpha Contrast) {materialFunction=&CheapContrast}
  n25 = DynamicParameter() {defaultValue=<rgba 1 1 1 0>, paramNames [Mask Alpha Lerp, Mask Lerp, "Mask Overlay Mode | Mult <> Add", Mask Overlay Lerp]}
  n26 = Saturate(None=LinearInterpolate(A=@n22, B=Multiply(A=@n22, B=@n24.Result), Alpha=@n25.Mask Alpha Lerp) {constA=1})
  Saturate(None=Multiply(A=@n1.A, B=Multiply(A=@n8.A, B=@n26)))
 ]
}
